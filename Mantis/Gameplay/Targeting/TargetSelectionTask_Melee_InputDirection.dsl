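; Melee targeting selection task that traces along the Fortnite player pawn's movement input to pick targets.
; GetSourceLocation starts the trace at the player camera targeting transform's location.
; GetTraceDirection uses the last input vector, or actor forward without input; GetSourceOffset combines that direction with the parent offset's components.
; Each override falls back to the parent implementation when its lookups fail, such as an invalid targeting subsystem.
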(fn GetSourceLocation
  (params (TargetingHandle :type Targeting%20Request%20Handle%20Structure%20%28by%20ref%29))
  (bind _returnvalue (Targeting|GetTargetingSubsystem TargetingHandle))
  ; GetSourceLocation: Get the camera targeting transform, use the returned location as the source location.
  (Utilities|FlowControl|Sequence
    (:then_0
      (Utilities|IsValid _returnvalue
        (:Is%20Valid
          (bind _asfort_player_pawn (Utilities|Casting|CastToFortPlayerPawn (Utilities|Struct|BreakTargetingSourceContext (TargetingSystem|TargetingTypes|GetTargetingSourceContext _returnvalue TargetingHandle))))
          (bind _returnvalue_2 (TargetingSystem|GetPlayerCameraTargetingTransform _asfort_player_pawn))
          (bind _returnvalue_location (Math|Transform|BreakTransform _returnvalue_2))
          (return _returnvalue_location))))
    (:then_1
      (bind _returnvalue_3 (|CallParentFunction TargetingHandle))
      (return _returnvalue_3))))

(fn GetSourceOffset
  (params (TargetingHandle :type Targeting%20Request%20Handle%20Structure%20%28by%20ref%29))
  (bind _returnvalue (Targeting|GetTargetingSubsystem TargetingHandle))
  ; GetSourceOffset: Transform world offset to relative offset to the trace direction.
  (Utilities|FlowControl|Sequence
    (:then_0
      (Utilities|IsValid _returnvalue
        (:Is%20Valid
          (bind _asfort_player_pawn (Utilities|Casting|CastToFortPlayerPawn (Utilities|Struct|BreakTargetingSourceContext (TargetingSystem|TargetingTypes|GetTargetingSourceContext _returnvalue TargetingHandle))))
          (bind _charactermovement (Variables|/Script/Engine.Character|GetCharacterMovement _asfort_player_pawn)
            (:then
              (if (Utilities|Operators|NotEqual(!=) :A (Math|Vector|Normalize (Pawn|Components|PawnMovement|GetLastInputVector _charactermovement)) :ErrorTolerance 0.01)
                (LocalVariables|SetNormalizedDirection (Pawn|Components|PawnMovement|GetLastInputVector _charactermovement))
                (else
                  (LocalVariables|SetNormalizedDirection (Transformation|GetActorForwardVector _asfort_player_pawn))))
              (bind _returnvalue_2 (|CallParentFunction (Variables|GetTargetingHandle)))
              (return (Utilities|Operators|Add (* (LocalVariables|GetNormalizedDirection) (.x _returnvalue_2)) (* (LocalVariables|GetNormalizedDirection) (.y _returnvalue_2)) (+ (LocalVariables|GetNormalizedDirection) (.z _returnvalue_2)))))
            (:else)))))
    (:then_1
      (bind _returnvalue_3 (|CallParentFunction TargetingHandle))
      (return _returnvalue_3))))

(fn GetTraceDirection
  (params (TargetingHandle :type Targeting%20Request%20Handle%20Structure%20%28by%20ref%29))
  (bind _returnvalue (Targeting|GetTargetingSubsystem TargetingHandle))
  ;[0.05784 0.04871 0.15 0.5] GetTraceDirection: Trace in the direction of movement input.
  (Utilities|FlowControl|Sequence
    (:then_0
      (Utilities|IsValid _returnvalue
        (:Is%20Valid
          (bind _asfort_player_pawn (Utilities|Casting|CastToFortPlayerPawn (Utilities|Struct|BreakTargetingSourceContext (TargetingSystem|TargetingTypes|GetTargetingSourceContext _returnvalue TargetingHandle))))
          (bind _charactermovement (Variables|/Script/Engine.Character|GetCharacterMovement _asfort_player_pawn)
            (:then
              (if (Utilities|Operators|NotEqual(!=) :A (Math|Vector|Normalize (Pawn|Components|PawnMovement|GetLastInputVector _charactermovement)) :ErrorTolerance 0.01)
                (return (Pawn|Components|PawnMovement|GetLastInputVector _charactermovement))
                (else
                  (return (Transformation|GetActorForwardVector _asfort_player_pawn)))))
            (:else)))))
    (:then_1
      (bind _returnvalue_2 (|CallParentFunction TargetingHandle))
      (return _returnvalue_2))))
